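# Material function that blends an animated ghost look, defaulting to Cat Burglar ghost textures and purple tints, over input material attributes by Alpha.
# It overrides base color, emissive and normal, combining scrolling triplanar normals, warped two-color swirl noise, Fresnel glow and a sine pulse.
# Pre-skinned normals and bounds-relative positions are computed for projection, though the bounds divisor subtracts Max from itself, yielding zero.
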MaterialFunction /Game/Global/Materials/MaterialGym/Materials/GhostNoise/MaterialFunctions/MF_GhostNoise
graph {nodes [
  n1 = NamedRerouteUsage()
  n2 = NamedRerouteUsage()
  n3 = NamedRerouteUsage()
  n4 = NamedRerouteUsage()
  FunctionOutput(None=BlendMaterialAttributes(
    A=@n1
    B=SetMaterialAttributes!(MaterialAttributes=@n1, BaseColor=@n2, EmissiveColor=@n3, Normal=@n4) {attributeSetTypes [#1, #2, #3]}
    Alpha=NamedRerouteUsage()
   ) {vertexAttributeBlendType=UseB})
  n5 = FunctionInput() {bUsePreviewValueAsDefault=true, inputName=Alpha, inputType=FunctionInput_Scalar, sortPriority=1, previewValue=<1 0 0 1>}
  NamedRerouteDeclaration(None=@n5) {name=Alpha, nodeColor=<rgba 1 0.7969 0 1>}
  n6 = LinearInterpolate(A=NamedRerouteUsage(), B=NamedRerouteUsage(), Alpha=NamedRerouteUsage())
  n7 = Saturate(None=Fresnel(
    ExponentIn=ScalarParameter() {defaultValue=2.5, group=Ghost Noise, parameterName=GhostFresnel_Exponent, sortPriority=11}
    BaseReflectFractionIn=Constant()
    Normal=Transform(None=@n6)
   ) {baseReflectFraction=0})
  n8 = VectorParameter() {group=Ghost Noise, parameterName=Ghost_FresnelColor, sortPriority=10, defaultValue=<rgba 0.1258 0.0113 0.14 0>}
  n9 = Add(A=@n8.RGB, B=NamedRerouteUsage())
  n10 = Add(
   A=NamedRerouteUsage()
   B=Multiply(A=Multiply(A=@n7, B=@n9), B=ScalarParameter() {defaultValue=1, group=Ghost Noise, parameterName=GhostFresnel_Strength, sortPriority=12})
  )
  NamedRerouteDeclaration(None=Add(A=NamedRerouteDeclaration(None=@n10) {name=GhostFresnelEmissive, nodeColor=<rgba 0.5703 0 1 1>}, B=NamedRerouteUsage())) {name=GhostEmissive, nodeColor=<rgba 0.125 1 0 1>}
  n11 = FunctionInput() {bUsePreviewValueAsDefault=true, inputName=Material, inputType=FunctionInput_MaterialAttributes}
  NamedRerouteDeclaration(None=GetMaterialAttributes(None=NamedRerouteDeclaration(None=@n11) {name=InMaterial, nodeColor=<rgba 0.9688 0 1 1>}) {attributeGetTypes [#3]}.Normal) {name=InNormal, nodeColor=<rgba 0 0.7187 1 1>}
  n12 = TextureObjectParameter() {
   group=Ghost Noise
   parameterName=GhostNormal_Texture
   samplerType=SAMPLERTYPE_Normal
   sortPriority=1
   texture=/Game/Characters/Player/Male/Medium/Bodies/M_MED_Cat_Burglar_Ghost/Textures/T_M_MED_Cat_Burglar_Ghost_N1
  }
  n13 = VectorParameter() {group=Ghost Noise, parameterName=GhostNormal_Speed, sortPriority=3, defaultValue=<rgba 0 0 -0.05 0>}
  n14 = Multiply(A=Time(), B=@n13.RGB)
  n15 = NamedRerouteUsage()
  n16 = MaterialFunctionCall!(
   TextureObject=@n12
   TextureSize=Constant() {r=1}
   RotYAxis
   RotZAxis
   RotXAxis
   PreSkinNormal=NamedRerouteUsage()
   ProjectionTransitionContrast=NamedRerouteUsage()
   WorldPosition=Multiply(A=Add(A=@n14, B=@n15), B=ScalarParameter() {defaultValue=10, group=Ghost Noise, parameterName=GhostNormal_Tile, sortPriority=2})
  ) {materialFunction=/Game/Packages/Fortress_SharedMaterials/Base_Material_Functions/BaseCustomizationFunctions/PreSkinnedAlignedNormal}
  n17 = Multiply(
   A=ComponentMask(None=@n16.XYZ Texture) {g=true, r=true}
   B=ScalarParameter() {defaultValue=1, group=Ghost Noise, parameterName=GhostNormal_Strength, sortPriority=3}
  )
  n18 = ComponentMask(None=@n16.XYZ Texture) {b=true}
  NamedRerouteDeclaration(None=MaterialFunctionCall(BaseNormal=NamedRerouteUsage(), AdditionalNormal=AppendVector(A=@n17, B=@n18)) {materialFunction=&BlendAngleCorrectedNormals}.Result) {name=GhostNormal, nodeColor=<rgba 0 0.3437 1 1>}
  n19 = VectorParameter() {group=Ghost Noise, parameterName=GhostNoise_Color_01, sortPriority=24, defaultValue=<rgba 0.0363 0.0045 0.0521 1>}
  n20 = VectorParameter() {group=Ghost Noise, parameterName=GhostNoise_Color_02, sortPriority=25, defaultValue=<rgba 0.5592 0.0683 1 1>}
  n21 = TextureObjectParameter() {
   group=Ghost Noise
   parameterName=GhostNoise_Texture
   samplerType=SAMPLERTYPE_LinearColor
   sortPriority=20
   texture=/Game/Global/Materials/MaterialGym/Materials/GhostNoise/Textures/T_ghost_swirls_01
  }
  n22 = Time()
  n23 = VectorParameter() {group=Ghost Noise, parameterName=GhostNoise_Speed, sortPriority=22, defaultValue=<rgba 0 0 -0.02 1>}
  n24 = Multiply(
   A=Add(A=Multiply(A=@n22, B=@n23.RGB), B=NamedRerouteUsage())
   B=VectorParameter() {group=Ghost Noise, parameterName=GhostNoise_Tile, sortPriority=21, defaultValue=<rgba 1 1 1 1>}.RGB
  )
  n25 = Multiply(A=ScalarParameter() {defaultValue=0.05, group=Ghost Noise, parameterName=GhostNoise_WarpAmount, sortPriority=23}, B=NamedRerouteUsage())
  n26 = ComponentMask(None=MaterialFunctionCall(TextureObject=@n21, TextureSize=Constant() {r=1}, WorldPosition=Add(A=@n24, B=@n25), ProjectionTransitionContrast=NamedRerouteUsage()) {materialFunction=&WorldAlignedTexture}.XYZ Texture) {r=true}
  NamedRerouteDeclaration(None=Multiply(
    A=LinearInterpolate(A=@n19.RGB, B=@n20.RGB, Alpha=@n26)
    B=ScalarParameter() {defaultValue=1, group=Ghost Noise, parameterName=GhostNoise_EmissiveStrength, sortPriority=26}
   )) {name=GhostNoise, nodeColor=<rgba 0.0547 1 0 1>}
  n27 = VectorParameter() {group=Ghost Noise, parameterName=Ghost_Base Color, sortPriority=40, defaultValue=<rgba 0.001 0.001 0.001 0.001>}
  NamedRerouteDeclaration(None=@n27.RGB) {name=GhostColor, nodeColor=<rgba 0 0.5078 1 1>}
  n28 = VectorParameter() {group=Ghost Noise, parameterName=Ghost_PulseColor, sortPriority=30, defaultValue=<rgba 0.1258 0.0113 0.14 0>}
  n29 = Time()
  n30 = ScalarParameter() {defaultValue=0.25, group=Ghost Noise, parameterName=Ghost_PulseSpeed, sortPriority=31}
  n31 = ConstantBiasScale(None=Sine(None=Multiply(A=@n29, B=@n30)))
  n32 = ScalarParameter() {defaultValue=1, group=Ghost Noise, parameterName=Ghost_PulseBrightness}
  NamedRerouteDeclaration(None=Multiply(A=@n28.RGB, B=Multiply(A=@n31, B=@n32))) {name=Pulse, nodeColor=<rgba 1 0 0.4687 1>}
  NamedRerouteDeclaration(None=VertexInterpolator(VS=PreSkinnedNormal()).PS) {name=PreskinnedNormal, nodeColor=<rgba 1 0.8672 0 1>}
  NamedRerouteDeclaration(None=ScalarParameter() {group=Ghost Noise, parameterName=GhostTransitionContrast, sortPriority=41}) {name=TransitionContrast, nodeColor=<rgba 0.2422 1 0 1>}
  n33 = VertexInterpolator(VS=PreSkinnedPosition())
  n34 = PreSkinnedLocalBounds()
  n35 = MaterialFunctionCall(Color=Subtract(A=@n34.Max, B=@n34.Max)) {materialFunction=&SplitComponents}
  n36 = Max(A=@n35.R, B=@n35.R)
  NamedRerouteDeclaration(None=Divide(A=Subtract(A=@n33.PS, B=@n34.Min), B=Max(A=@n36, B=@n35.B))) {name=UVBoundsMax0_1, nodeColor=<rgba 1 0.2109 0 1>}
 ]}
