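# M_GhostNoise is a Default Lit material, usable on skeletal and mesh-deformer meshes, that applies a ghost noise effect through MF_GhostNoise.
# It builds simple base attributes from a constant color, a Roughness parameter and a flat normal.
# A Fresnel term on the world-space vertex normal, with parameterized exponent and base reflect fraction, feeds the function's Alpha input alongside those attributes.
# The noise logic itself lives in the material function.
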Material /Game/Global/Materials/MaterialGym/Materials/GhostNoise/M_GhostNoise
graph {
 properties {
  bUseMaterialAttributes=true
  bUsedWithMeshDeformer=true
  bUsedWithSkeletalMesh=true
  pixelDepthOffsetMode=PDOM_Legacy
  usedShadingModels=Default Lit
 }
 nodes [
  Constant()
  n1 = Constant3Vector() {constant=<rgba 0 0 1 0>}
  MaterialFunctionCall!(
   Material=MakeMaterialAttributes!(
    BaseColor=Constant3Vector()
    Metallic
    Specular
    Roughness=ScalarParameter() {defaultValue=1, parameterName=Roughness}
    Anisotropy
    EmissiveColor
    Opacity
    OpacityMask
    Normal=@n1
    Tangent
    WorldPositionOffset
    SubsurfaceColor
    ClearCoat
    ClearCoatRoughness
    AmbientOcclusion
    Refraction
    CustomizedUVs_0
    CustomizedUVs_1
    CustomizedUVs_2
    CustomizedUVs_3
    CustomizedUVs_4
    CustomizedUVs_5
    CustomizedUVs_6
    CustomizedUVs_7
    PixelDepthOffset
    ShadingModel
    Displacement
   )
   Alpha=Fresnel(
    ExponentIn=ScalarParameter() {defaultValue=1.5, parameterName=ExponentIn}
    BaseReflectFractionIn=ScalarParameter() {defaultValue=0.1, parameterName=BaseReflectFractionIn}
    Normal=VertexNormalWS()
   )
  ) {materialFunction=/Game/Global/Materials/MaterialGym/Materials/GhostNoise/MaterialFunctions/MF_GhostNoise}
  Transform(None=@n1)
 ]
}
